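# Test overlay material for weapon wraps that draws a crystalline wrap effect as a masked, two-sided layer for skinned meshes.
# A Masks texture sample supplies RGB to the customization mask input of MF_Customization_Crystalline, which runs on otherwise empty material attributes.
# The same sample's red channel sets opacity mask, clipping the layer to the masked region.
# World position offset pushes vertices along their normals by the WPO parameter, lifting the overlay shell.
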Material /Game/NeverCook/ArtTools/OverlayWeaponWraps/Tests/M_Overlay
graph {
 properties {
  bAutomaticallySetUsageInEditor=false
  bUseMaterialAttributes=true
  bUsedWithClothing=true
  bUsedWithMeshDeformer=true
  bUsedWithMorphTargets=true
  bUsedWithSkeletalMesh=true
  blendMode=BLEND_Masked
  translucencyLightingMode=TLM_SurfacePerPixelLighting
  twoSided=true
  usedShadingModels=Default Lit
 }
 nodes [
  n1 = TextureSample() {samplerType=SAMPLERTYPE_Masks}
  SetMaterialAttributes!(
   MaterialAttributes=MaterialFunctionCall!(
    Material=SetMaterialAttributes!(MaterialAttributes)
    ------------------
    PreSkinLocalPosition
    PreSkinnedLocalNormal
    "Customization Crevasse (_CM G)"
    "Customization Mask (_CM B)"=@n1.RGB
   ) {materialFunction=/Game/Weapons/WeaponSkins/Wraps/MaterialFunctions/MF_Customization_Crystalline}.Result
   World Position Offset=Multiply(A=VertexNormalWS(), B=ScalarParameter() {defaultValue=0.01, parameterName=WPO})
   Opacity Mask=@n1.R
  ) {attributeSetTypes [#1, #2]}
 ]
}
