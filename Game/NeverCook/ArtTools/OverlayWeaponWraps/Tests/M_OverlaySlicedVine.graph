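# Masked test material for an overlay weapon wrap on skeletal meshes, built from the shared wrap material functions.
# MF_TopSideProjection_Wrap_V02 builds the surface attributes, which feed MF_Effect_Decal_Reactive alongside adjustable wrap bounds and pre-skinned local normals.
# The overlay is pushed out along vertex normals by a WPO parameter and clipped by the CustomizationMask red channel.
# An iridescent effect call and some pre-skinned bounds math are left unconnected.
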Material /Game/NeverCook/ArtTools/OverlayWeaponWraps/Tests/M_OverlaySlicedVine
graph {
 properties {
  bAutomaticallySetUsageInEditor=false
  bUseMaterialAttributes=true
  bUsedWithMeshDeformer=true
  bUsedWithSkeletalMesh=true
  blendMode=BLEND_Masked
  usedShadingModels=Default Lit
 }
 nodes [
  MaterialFunctionCall!(
   Material
   Decal_Mask
   Mask Crevise
   "Customization Mask (_CM B)"
   Mask Primary
   Mask Secondary
   TS_Mask_01
   TS_Mask_03
   TS_Mask_04
   TS_Mask_02
  ) {materialFunction=/Game/Weapons/WeaponSkins/Wraps/MaterialFunctions/MF_Effect_Iridescent}
  n1 = Constant()
  n2 = MaterialFunctionCall!(
   is Vehicle=@n1
   Object Bounds Max Adjustment=VectorParameter() {parameterName=OBMAX, defaultValue=<rgba 0 0 0 1>}.RGB
   Object Bounds Min Adjustment=VectorParameter() {parameterName=OBMIN, defaultValue=<rgba 0 0 0 1>}.RGB
   Object Bounds Center Adjustment=VectorParameter() {parameterName=OBCENTER}.RGB
   isMech=@n1
   Side Project Vehicle=@n1
   Local Position=VertexInterpolator(VS=LocalPosition().XYZ).PS
  ) {materialFunction=/Game/Weapons/WeaponSkins/Wraps/MaterialFunctions/MF_WrapBounds}
  NamedRerouteDeclaration(None=@n2.Result) {name=Bounds, nodeColor=<rgba 1 0.5625 0 1>}
  n3 = MaterialFunctionCall() {materialFunction=/Game/Packages/Fortress_SharedMaterials/Base_Material_Functions/BaseCustomizationFunctions/MF_FN_Dummy_MASTER}
  NamedRerouteDeclaration(None=@n3.Pre-Skinned Local Normal Pixel Shader) {name=PreSkinnedLocalNormal, nodeColor=<rgba 0.2422 0 1 1>}
  NamedRerouteDeclaration(None=@n3.Pre-Skinned Local Position Pixel Shader) {name=PreskinnedLocalPosition, nodeColor=<rgba 0.0781 1 0 1>}
  n4 = TextureSampleParameter2D() {parameterName=CustomizationMask, samplerType=SAMPLERTYPE_Masks, texture=null}
  NamedRerouteDeclaration(None=@n4.RGBA) {name=Mask, nodeColor=<rgba 1 0 0.3984 1>}
  n5 = SetMaterialAttributes!(MaterialAttributes)
  n6 = NamedRerouteUsage()
  n7 = MaterialFunctionCall!(
   Triplaners-Masks textures=Constant3Vector() {constant=<rgba 1 1 1 0>}
   Material=@n5
   Root_Material=@n5
   IsState
   IsState_B
   Pre-Skinned Local Normal=NamedRerouteUsage()
   Calcated_Bounds=NamedRerouteUsage()
   Local Position=NamedRerouteUsage()
   Mask Secondary=ComponentMask(None=@n6) {b=true}
   Mask Crevise
   TS Mask_01
   TS Mask_02
   Mask Primary=ComponentMask(None=@n6) {r=true}
   Object-SafeZone-Back
   Object-SafeZone-Middle
   Liquid Mask Input
   Object-SafeZone-Front
  ) {materialFunction=/Game/Weapons/WeaponSkins/Wraps/MaterialFunctions/MF_TopSideProjection_Wrap_V02}
  SetMaterialAttributes!(
   MaterialAttributes=MaterialFunctionCall!(
    Root_Matieral=@n7.Result
    Material=@n7.Result
    "Msk_Ts01<--->Ts02"
    IsState_B
    IsState
    CM_Msk=@n4.RGBA
    Pre-Skinned Local Normal=@n3.Pre-Skinned Local Normal Pixel Shader
    Calcated_Bounds=@n2.Result
    Object-SafeZone-Back
    Object-SafeZone-Middle
    Object-SafeZone-Front
   ) {materialFunction=/Game/Weapons/WeaponSkins/Wraps/MaterialFunctions/MF_Effect_Decal_Reactive}.Result
   World Position Offset=Multiply(A=VertexNormalWS(), B=ScalarParameter() {defaultValue=0.01, parameterName=WPO})
   Opacity Mask=@n4.R
  ) {attributeSetTypes [#1, #2]}
  n8 = PreSkinnedLocalBounds()
  Subtract(A=VertexInterpolator(VS=PreSkinnedPosition()).PS, B=@n8.Min)
  Subtract(A=@n8.Max, B=@n8.Max)
 ]
}
